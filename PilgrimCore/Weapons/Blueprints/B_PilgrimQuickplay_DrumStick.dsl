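; Drum stick weapon Blueprint for Pilgrim Quickplay that tags its wielder while equipped and applies cosmetic variants to the instrument.
; On equip and attach, it adds the loose gameplay tags from its GameplayTagsToAdd variable to the instigator, then applies variants through SparksCosmeticsBPFL.
; On unequip and detach, it removes the same tags before calling the parent.
; BeginPlay, Tick and the construction script add no logic of their own.
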(event EventBeginPlay
  (|CallParentFunction) ; This node is disabled and will not be called.\nDrag off pins to build functionality.
)

(event EventTick
  (params (DeltaSeconds :type Float%20%28single-precision%29)))

(event FortWeapon|EventOnEquippedAndAttached
  (|CallParentFunction)
  (Ability|GameplayEffect|AddLooseGameplayTags (Game|GetInstigator) (Variables|Default|GetGameplayTagsToAdd))
  (Class|SparksCosmeticsBPFL|ApplyVariantstoInstrument self))

(event FortWeapon|EventOnUnEquippedAndDetached
  (Ability|GameplayEffect|RemoveLooseGameplayTags (Game|GetInstigator) (Variables|Default|GetGameplayTagsToAdd))
  (|CallParentFunction))

(fn ConstructionScript
  (|CallParentFunction))
