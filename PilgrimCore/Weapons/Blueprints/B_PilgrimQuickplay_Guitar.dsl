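; Fortnite weapon Blueprint for the Pilgrim Quickplay guitar, handling character tags, animation and cosmetics as the instrument is equipped or unequipped.
; On equip it adds loose gameplay tags from a variable to the instigator, then applies instrument variants through SparksCosmeticsBPFL.
; Characters whose loadout cosmetic carries the Skirts emote meta tag also get a skirt-specific guitar animset override.
; On unequip it removes the same tags before calling the parent.
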(event FortWeapon|EventOnEquippedAndAttached
  (bind _returnvalue (Game|GetInstigator))
  (|CallParentFunction)
  (Ability|GameplayEffect|AddLooseGameplayTags _returnvalue (Variables|Default|GetGameplayTagsToAdd))
  (bind _asfort_player_pawn (Utilities|Casting|CastToFortPlayerPawn _returnvalue))
  (Utilities|FlowControl|Sequence
    (:then_0
      (if (CosmeticLoadout|DoesItemDefinitionfromActorHaveMetaTag _asfort_player_pawn "(TagName=\"LoadoutSlot.Character\")" "(TagName=\"Cosmetics.Emote.CharacterGroup.Skirts\")")
        (CharacterAnimation|SetWeaponPawnAnimsetOverride "/SparksCharacterFull/Animation/MainPlayer/AnimSets/Guitar_AnimSet_Skirts.Guitar_AnimSet_Skirts")))
    (:then_1
      (Class|SparksCosmeticsBPFL|ApplyVariantstoInstrument self))))

(event FortWeapon|EventOnUnEquippedAndDetached
  (Ability|GameplayEffect|RemoveLooseGameplayTags (Game|GetInstigator) (Variables|Default|GetGameplayTagsToAdd))
  (|CallParentFunction))

(fn ConstructionScript
  (|CallParentFunction))
