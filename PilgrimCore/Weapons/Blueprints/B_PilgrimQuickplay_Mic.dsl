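; Fortnite weapon Blueprint for the Pilgrim quickplay mic, which swaps the wielder's search montage and tags while equipped.
; On equip it adds loose gameplay tags, caches and replaces the pawn's OverrideSearchMontageInfo, applies the UseOverrideSearchMontage tag effect and applies Sparks cosmetic variants to itself.
; On unequip it removes the tags and stored effect handle's effect, restores the cached montage info and clears the cache.
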(event EventBeginPlay
  (|CallParentFunction) ; This node is disabled and will not be called.\nDrag off pins to build functionality.
)

(event EventTick
  (params (DeltaSeconds :type Float%20%28single-precision%29)))

(event FortWeapon|EventOnEquippedAndAttached
  (bind _returnvalue (Game|GetInstigator))
  (|CallParentFunction)
  (Ability|GameplayEffect|AddLooseGameplayTags _returnvalue (Variables|Default|GetGameplayTagsToAdd))
  (bind _asplayer_pawn_athena (Utilities|Casting|CastToPlayerPawn_Athena _returnvalue))
  (Utilities|FlowControl|Sequence
    (:then_0
      (Variables|Default|SetCachedOverrideSearchMontageInfo (Variables|/Game/Athena/PlayerPawn_Athena.PlayerPawn_Athena_C|GetOverrideSearchMontageInfo _asplayer_pawn_athena)))
    (:then_1
      (Variables|/Game/Athena/PlayerPawn_Athena.PlayerPawn_Athena_C|SetOverrideSearchMontageInfo (Variables|Default|GetOverrideSearchMontageInfo) _asplayer_pawn_athena))
    (:then_2
      (Utilities|IsValid (Ability|GetFortAbilitySystemComponent _asplayer_pawn_athena)
        (:Is%20Valid
          (bind _returnvalue_2 (GameplayEffects|ApplyGameplayEffectToSelf :self (Ability|GetFortAbilitySystemComponent _asplayer_pawn_athena) :GameplayEffectClass "/Game/Abilities/Player/Generic/Traits/DefaultPlayer/GE_ApplyTag_UseOverrideSearchMontage.GE_ApplyTag_UseOverrideSearchMontage_C" :Level 0.0))
          (Variables|Default|SetActiveGameplayEffectHandle _returnvalue_2))))
    (:then_3
      (Class|SparksCosmeticsBPFL|ApplyVariantstoInstrument self))))

(event FortWeapon|EventOnUnEquippedAndDetached
  (bind _returnvalue (Game|GetInstigator))
  (Utilities|FlowControl|Sequence
    (:then_0
      (Ability|GameplayEffect|RemoveLooseGameplayTags _returnvalue (Variables|Default|GetGameplayTagsToAdd))
      (bind _asplayer_pawn_athena (Utilities|Casting|CastToPlayerPawn_Athena _returnvalue))
      (Utilities|FlowControl|Sequence
        (:then_0
          (GameplayEffects|RemoveActiveGameplayEffect (Ability|GetFortAbilitySystemComponent _asplayer_pawn_athena) (Variables|Default|GetActiveGameplayEffectHandle) -1))
        (:then_1
          (Variables|/Game/Athena/PlayerPawn_Athena.PlayerPawn_Athena_C|SetOverrideSearchMontageInfo (Variables|Default|GetCachedOverrideSearchMontageInfo) _asplayer_pawn_athena))
        (:then_2
          (Variables|Default|SetCachedOverrideSearchMontageInfo))))
    (:then_1
      (|CallParentFunction))))

(fn ConstructionScript
  (|CallParentFunction))
